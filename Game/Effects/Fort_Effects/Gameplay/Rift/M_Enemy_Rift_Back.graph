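# Translucent, unlit, two-sided mesh-particle material for the back layer of the enemy rift effect.
# Two tornado-tile layers pan at different speeds, one offset and dimmed, then sum and take particle color before MF_FX_EmissiveAdjustments.
# Two counter-panning smoke ring masks multiply, get a slight boost, and scale by particle alpha to form opacity.
# Depth fade softens that opacity where it meets scene geometry.
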Material /Game/Effects/Fort_Effects/Gameplay/Rift/M_Enemy_Rift_Back
graph {
 properties {
  bUsedWithMeshParticles=true
  blendMode=BLEND_Translucent
  pixelDepthOffsetMode=PDOM_Legacy
  shadingModel=MSM_Unlit
  twoSided=true
  usedShadingModels=Unlit
 }
 nodes [
  n1 = ParticleColor()
  n2 = Panner(Coordinate=TextureCoordinate() {uTiling=2}) {speedX=-0.1}
  n3 = Panner(Coordinate=TextureCoordinate()) {speedX=0.1}
  n4 = Multiply(
   A=Multiply(
    A=TextureSample(UVs=@n2) {texture=/Game/Effects/Fort_Effects/Textures/T_Smoke_Ring_Mask2}.G
    B=TextureSample(UVs=@n3) {texture=&T_Smoke_Ring_Mask2}.RGB
   )
   B=Constant() {r=1.2}
  )
  DepthFade(Opacity=Clamp(None=Multiply(A=@n1.A, B=@n4))) {fadeDistanceDefault=200}
  n5 = Add(A=TextureCoordinate() {uTiling=2, vTiling=3}, B=AppendVector(A=Constant() {r=5.5}, B=Constant() {r=6.5}))
  n6 = Multiply(
   A=Constant() {r=0.4}
   B=TextureSample(UVs=Panner(Coordinate=@n5) {speedX=-0.02, speedY=0.12}) {texture=/Game/Global/Textures/Default/Tileable_Patterns/T_Tornado_Tile}.G
  )
  n7 = TextureSample(UVs=Panner(Coordinate=TextureCoordinate() {uTiling=2, vTiling=3}) {speedX=-0.04, speedY=0.5}) {texture=&T_Tornado_Tile}
  MaterialFunctionCall(Emissive=Multiply(A=Add(A=@n6, B=@n7.G), B=@n1.RGB)) {materialFunction=&MF_FX_EmissiveAdjustments}
 ]
}
